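# Sound cue for the Pizza Party item's box pre-throw sound, classed as weapon foley and using weapon reload attenuation.
# A mixer layers two random picks: wave 01 or a pitched-up wave 02, plus wave 03 or 04 at reduced volume.
# A root modulator adds slight pitch variation to the mix, and the cue itself plays quietly at a lowered pitch.
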SoundCue /FlipperGameplay/Items/PizzaParty/Audio/PizzaParty_BoxPreThrow_Cue
root=Modulator_0
nodes [
 {id=Mixer_0, children [Random_1,Random_0], properties {inputVolume [1,0.7]}}
 {id=Modulator_0, children [Mixer_0], properties {pitchMax=1.02, pitchMin=0.98}}
 {id=Modulator_1, children [WavePlayer_3], properties {pitchMax=1.15, pitchMin=1.05}}
 {id=Random_0, children [WavePlayer_0,WavePlayer_2], properties {weights [1,1]}}
 {id=Random_1, children [WavePlayer_1,Modulator_1], properties {weights [1,1]}}
 {id=WavePlayer_0, properties {soundWaveAssetPtr=&/PizzaParty_BoxPreThrow_03}}
 {id=WavePlayer_1, properties {soundWaveAssetPtr=&/PizzaParty_BoxPreThrow_01}}
 {id=WavePlayer_2, properties {soundWaveAssetPtr=&/PizzaParty_BoxPreThrow_04}}
 {id=WavePlayer_3, properties {soundWaveAssetPtr=&/PizzaParty_BoxPreThrow_02}}
]
properties {
 attenuationSettings=&Fort_Weapon_Reload_Att
 maxDistance=2250
 pitchMultiplier=0.9
 soundClassObject=&Fort_WeaponFoley
 volumeMultiplier=0.12
}
