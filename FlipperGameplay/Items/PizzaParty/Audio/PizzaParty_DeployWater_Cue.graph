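# Sound cue for the Pizza Party item's deploy on water, judging by its name and splash layer.
# A mixer layers random box-impact, squish and splash variants, all pitched up or pitch-randomized, with a delayed Italian box-equip sound; splashes are boosted in the mix.
# One squish variant gets extra pitch, the other a short delay.
# It reuses bubble shield attenuation and overrides concurrency to stop the oldest instance beyond a small limit.
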SoundCue /FlipperGameplay/Items/PizzaParty/Audio/PizzaParty_DeployWater_Cue
root=Mixer_0
nodes [
 {id=Delay_0, children [WavePlayer_5], properties {delayMax=0.43, delayMin=0.43}}
 {id=Delay_1, children [WavePlayer_8], properties {delayMax=0.08, delayMin=0.08}}
 {id=Mixer_0, children [Modulator_0,Delay_0,Modulator_1,Modulator_3], properties {inputVolume [0.6,0.8,1.6,1]}}
 {id=Modulator_0, children [Random_0], properties {pitchMax=1.1, pitchMin=1}}
 {id=Modulator_1, children [Random_1], properties {pitchMax=1.1, pitchMin=1.1}}
 {id=Modulator_2, children [WavePlayer_7], properties {pitchMax=1.1, pitchMin=1.1}}
 {id=Modulator_3, children [Random_3], properties {pitchMax=1.2, pitchMin=1.1}}
 {id=Random_0, children [WavePlayer_3,WavePlayer_4], properties {weights [1,1]}}
 {id=Random_1, children [Modulator_2,Delay_1], properties {weights [1,1]}}
 {id=Random_3, children [WavePlayer_11, WavePlayer_12, WavePlayer_13], properties {weights [1,1,1]}}
 {id=WavePlayer_11, properties {soundWaveAssetPtr=&/PizzaParty_Splash_01}}
 {id=WavePlayer_12, properties {soundWaveAssetPtr=&/PizzaParty_Splash_02}}
 {id=WavePlayer_13, properties {soundWaveAssetPtr=&/PizzaParty_Splash_03}}
 {id=WavePlayer_3, properties {soundWaveAssetPtr=&/PizzaParty_BoxThrowImpact_01}}
 {id=WavePlayer_4, properties {soundWaveAssetPtr=&/PizzaParty_BoxThrowImpact_02}}
 {id=WavePlayer_5, properties {soundWaveAssetPtr=&/PizzaParty_BoxEquip_Italian}}
 {id=WavePlayer_7, properties {soundWaveAssetPtr=&/PizzaParty_BoxThrowImpact_Squish_01}}
 {id=WavePlayer_8, properties {soundWaveAssetPtr=&/PizzaParty_BoxThrowImpact_Squish_02}}
]
properties {
 attenuationSettings=/Game/Sounds/Attenuation/Fort_Weapon_Bubble_Shield_Open_Close_Att
 bOverrideConcurrency=true
 maxDistance=10800
 soundClassObject=&Fort_WeaponRanged_Fire_3P_PSM
 volumeMultiplier=0.25
 concurrencyOverrides {maxCount=3, resolutionRule=StopOldest}
}
