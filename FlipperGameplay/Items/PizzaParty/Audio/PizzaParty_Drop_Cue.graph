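# Sound cue that plays a pizza box drop sound for the PizzaParty item.
# A random node picks one of three equally weighted box drop wave players.
# It uses the shared weapon drop attenuation and loot collect sound class, with a slightly raised pitch and reduced volume.
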SoundCue /FlipperGameplay/Items/PizzaParty/Audio/PizzaParty_Drop_Cue
root=Random_0
nodes [
 {id=Random_0, children [WavePlayer_1,WavePlayer_2,WavePlayer_3], properties {weights [1,1,1]}}
 {id=WavePlayer_1, properties {soundWaveAssetPtr=&/PizzaParty_BoxDrop_01}}
 {id=WavePlayer_2, properties {soundWaveAssetPtr=&/PizzaParty_BoxDrop_02}}
 {id=WavePlayer_3, properties {soundWaveAssetPtr=&/PizzaParty_BoxDrop_03}}
]
properties {
 attenuationSettings=&Fort_Weapon_Drop_Att
 maxDistance=3000
 pitchMultiplier=1.1
 soundClassObject=&Fort_Loot_Collect
 volumeMultiplier=0.5
}
