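# Sound cue for eating a Pizza Party slice, pairing an eating sound with a health boost sting.
# A Random node picks one of five equally weighted slice waves, mixed with the boost wave after a short fixed delay.
# Modulators on the boost branch and the final mix hold pitch constant.
# It uses consumable attenuation and sound class, with a concurrency override that stops the oldest instance at the limit.
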SoundCue /FlipperGameplay/Items/PizzaParty/Audio/PizzaParty_Eat_Cue
root=Modulator_0
nodes [
 {id=Delay_0, children [WavePlayer_10], properties {delayMax=0.1, delayMin=0.1}}
 {id=Mixer_1, children [Random_0,Modulator_1], properties {inputVolume [1.3,1.2]}}
 {id=Modulator_0, children [Mixer_1], properties {pitchMax=1, pitchMin=1}}
 {id=Modulator_1, children [Delay_0], properties {pitchMax=1, pitchMin=1}}
 {id=Random_0, children [WavePlayer_2,WavePlayer_3,WavePlayer_4,WavePlayer_5,WavePlayer_6], properties {weights [1,1,1,1,1]}}
 {id=WavePlayer_10, properties {soundWaveAssetPtr=&/PizzaParty_EatSlice_HealthBoost01}}
 {id=WavePlayer_2, properties {soundWaveAssetPtr=&/PizzaParty_EatSlice_01}}
 {id=WavePlayer_3, properties {soundWaveAssetPtr=&/PizzaParty_EatSlice_02}}
 {id=WavePlayer_4, properties {soundWaveAssetPtr=&/PizzaParty_EatSlice_03}}
 {id=WavePlayer_5, properties {soundWaveAssetPtr=&/PizzaParty_EatSlice_04}}
 {id=WavePlayer_6, properties {soundWaveAssetPtr=&/PizzaParty_EatSlice_05}}
]
properties {
 attenuationSettings=&Fort_Consumable_Use_Athena_Att
 bOverrideConcurrency=true
 maxDistance=2250
 soundClassObject=&Fort_Consumables
 volumeMultiplier=2.5
 concurrencyOverrides {maxCount=4, resolutionRule=StopOldest}
}
