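# Sound cue for equipping the Pizza Party item, layering a box-handling sound with a secondary "Italian" sound.
# A mixer combines one of three equally weighted, modulated box equip variations with the quieter Italian wave, which starts after a short fixed delay.
# It uses weapon reload attenuation and the consumables sound class, and overrides concurrency to cap simultaneous instances.
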SoundCue /FlipperGameplay/Items/PizzaParty/Audio/PizzaParty_Equip_Cue
root=Mixer_1
nodes [
 {id=Delay_0, children [WavePlayer_1], properties {delayMax=0.2, delayMin=0.2}}
 {id=Mixer_1, children [Modulator_0,Delay_0], properties {inputVolume [1.1,0.6]}}
 {id=Modulator_0, children [Random_0]}
 {id=Random_0, children [WavePlayer_0,WavePlayer_2,WavePlayer_3], properties {weights [1,1,1]}}
 {id=WavePlayer_0, properties {soundWaveAssetPtr=&/PizzaParty_BoxEquip_01}}
 {id=WavePlayer_1, properties {soundWaveAssetPtr=&/PizzaParty_BoxEquip_Italian}}
 {id=WavePlayer_2, properties {soundWaveAssetPtr=&/PizzaParty_BoxEquip_02}}
 {id=WavePlayer_3, properties {soundWaveAssetPtr=&/PizzaParty_BoxEquip_03}}
]
properties {
 attenuationSettings=&Fort_Weapon_Reload_Att
 bOverrideConcurrency=true
 maxDistance=2250
 soundClassObject=&Fort_Consumables
 volumeMultiplier=0.9
 concurrencyOverrides {maxCount=3}
}
