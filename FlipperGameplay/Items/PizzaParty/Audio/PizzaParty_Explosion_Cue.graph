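# Sound cue for the Pizza Party item's explosion, built from box-destroy sounds.
# A Random node picks one of three equally weighted waves and feeds it through a Modulator at the root.
# The cue lowers pitch and volume, routes to the gadgets sound class and borrows the toy-hit attenuation.
# Its concurrency override prevents new instances from starting once the limit is reached.
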SoundCue /FlipperGameplay/Items/PizzaParty/Audio/PizzaParty_Explosion_Cue
root=Modulator_0
nodes [
 {id=Modulator_0, children [Random_2]}
 {id=Random_2, children [WavePlayer_14, WavePlayer_15, WavePlayer_16], properties {weights [1,1,1]}}
 {id=WavePlayer_14, properties {soundWaveAssetPtr=&/PizzaParty_BoxDestroy_01}}
 {id=WavePlayer_15, properties {soundWaveAssetPtr=&/PizzaParty_BoxDestroy_02}}
 {id=WavePlayer_16, properties {soundWaveAssetPtr=&/PizzaParty_BoxDestroy_03}}
]
properties {
 attenuationSettings=/Game/Sounds/Attenuation/Fort_ToyHit_Att
 bOverrideConcurrency=true
 maxDistance=2800
 pitchMultiplier=0.8
 soundClassObject=&Fort_Gadgets
 volumeMultiplier=0.6
 concurrencyOverrides {resolutionRule=PreventNew}
}
