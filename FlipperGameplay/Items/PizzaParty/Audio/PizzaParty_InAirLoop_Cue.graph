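# Sound cue for the looping in-air sound of the Pizza Party item's box, in the gadgets sound class.
# A Random node picks one of two equally weighted looping waves, and a Modulator applies slight upward pitch variation on top of the cue's raised pitch.
# Inline overrides set logarithmic distance falloff and an owner-limited concurrency cap with volume scaling for overlapping instances.
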SoundCue /FlipperGameplay/Items/PizzaParty/Audio/PizzaParty_InAirLoop_Cue
root=Modulator_1
nodes [
 {id=Modulator_1, children [Random_0], properties {pitchMax=1.2, pitchMin=1}}
 {id=Random_0, children [WavePlayer_1,WavePlayer_2], properties {weights [1,1]}}
 {id=WavePlayer_1, properties {bLooping=true, soundWaveAssetPtr=&/PizzaParty_BoxInAirLoop_01}}
 {id=WavePlayer_2, properties {bLooping=true, soundWaveAssetPtr=&/PizzaParty_BoxInAirLoop_02}}
]
properties {
 attenuationSettings=/Game/Sounds/Attenuation/Fort_ToyHit_Att
 bOverrideAttenuation=true
 bOverrideConcurrency=true
 maxDistance=2000
 pitchMultiplier=1.2
 soundClassObject=&Fort_Gadgets
 volumeMultiplier=0.8
 attenuationOverrides {distanceAlgorithm=Logarithmic, falloffDistance=2000, stereoSpread=512, attenuationShapeExtents {x=0}}
 concurrencyOverrides {bLimitToOwner=true, maxCount=2, volumeScale=0.7}
}
