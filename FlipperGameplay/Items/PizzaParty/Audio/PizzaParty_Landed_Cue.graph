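# Sound cue for the Pizza Party item landing, playing a layered box-drop impact.
# A mixer combines a random impact, a random impact layer and a quieter squish from the box-throw set, delayed by a fixed short time.
# A root modulator slightly varies the pitch of the whole mix.
# It reuses Corn bounce attenuation, uses the loot-collect sound class and overrides concurrency to block new instances beyond a small limit.
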SoundCue /FlipperGameplay/Items/PizzaParty/Audio/PizzaParty_Landed_Cue
root=Modulator_5
nodes [
 {id=Delay_0, children [Modulator_0], properties {delayMax=0.15, delayMin=0.15}}
 {id=Mixer_0, children [Random_0,Random_2,Delay_0], properties {inputVolume [1,1.1,0.3]}}
 {id=Modulator_0, children [WavePlayer_6], properties {pitchMax=1, pitchMin=1}}
 {id=Modulator_5, children [Mixer_0], properties {pitchMax=1, pitchMin=0.9}}
 {id=Random_0, children [WavePlayer_0,WavePlayer_1,WavePlayer_2], properties {weights [1,1,1]}}
 {id=Random_2, children [WavePlayer_3,WavePlayer_4,WavePlayer_5], properties {weights [1,1,1]}}
 {id=WavePlayer_0, properties {soundWaveAssetPtr=&/PizzaParty_BoxDropImpact_01}}
 {id=WavePlayer_1, properties {soundWaveAssetPtr=&/PizzaParty_BoxDropImpact_02}}
 {id=WavePlayer_2, properties {soundWaveAssetPtr=&/PizzaParty_BoxDropImpact_03}}
 {id=WavePlayer_3, properties {soundWaveAssetPtr=&/PizzaParty_BoxDropImpactLayer_01}}
 {id=WavePlayer_4, properties {soundWaveAssetPtr=&/PizzaParty_BoxDropImpactLayer_02}}
 {id=WavePlayer_5, properties {soundWaveAssetPtr=&/PizzaParty_BoxDropImpactLayer_03}}
 {id=WavePlayer_6, properties {soundWaveAssetPtr=&/PizzaParty_BoxThrowImpact_Squish_01}}
]
properties {
 attenuationSettings=/Game/Athena/Sounds/Items/Corn/Fort_Corn_Bounce_Att
 bOverrideConcurrency=true
 maxDistance=3000
 soundClassObject=&Fort_Loot_Collect
 volumeMultiplier=0.4
 concurrencyOverrides {maxCount=3, resolutionRule=PreventNew}
}
