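# Sound cue for picking up the Pizza Party item, playing a box pickup sound.
# A Random node picks one of three equally weighted wave variations on each play.
# It uses weapon pickup attenuation and the gadget items sound class at reduced volume, with a concurrency override that prevents new instances once the limit is reached.
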SoundCue /FlipperGameplay/Items/PizzaParty/Audio/PizzaParty_Pickup_Cue
root=Random_0
nodes [
 {id=Random_0, children [WavePlayer_0,WavePlayer_2,WavePlayer_3], properties {weights [1,1,1]}}
 {id=WavePlayer_0, properties {soundWaveAssetPtr=&/PizzaParty_BoxPickup_01}}
 {id=WavePlayer_2, properties {soundWaveAssetPtr=&/PizzaParty_BoxPickup_02}}
 {id=WavePlayer_3, properties {soundWaveAssetPtr=&/PizzaParty_BoxPickup_03}}
]
properties {
 attenuationSettings=&Fort_Weapon_Pickup_Att
 bOverrideConcurrency=true
 maxDistance=3500
 soundClassObject=&Fort_Gadgets_Items
 volumeMultiplier=0.2
 concurrencyOverrides {resolutionRule=PreventNew}
}
