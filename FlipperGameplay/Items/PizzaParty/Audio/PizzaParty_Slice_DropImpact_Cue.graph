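# Sound cue for the impact of a Pizza Party slice item being dropped.
# A Random node picks one of four equally weighted drop-impact waves and feeds a Modulator at the root.
# It uses the weapon-drop attenuation and loot-collect sound class, with reduced overall volume.
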SoundCue /FlipperGameplay/Items/PizzaParty/Audio/PizzaParty_Slice_DropImpact_Cue
root=Modulator_1
nodes [
 {id=Modulator_1, children [Random_1]}
 {id=Random_1, children [WavePlayer_0,WavePlayer_3,WavePlayer_5,WavePlayer_6], properties {weights [1,1,1,1]}}
 {id=WavePlayer_0, properties {soundWaveAssetPtr=&/PizzaParty_SliceDropImpact_01}}
 {id=WavePlayer_3, properties {soundWaveAssetPtr=&/PizzaParty_SliceDropImpact_02}}
 {id=WavePlayer_5, properties {soundWaveAssetPtr=&/PizzaParty_SliceDropImpact_04}}
 {id=WavePlayer_6, properties {soundWaveAssetPtr=&/PizzaParty_SliceDropImpact_05}}
]
properties {
 attenuationSettings=&Fort_Weapon_Drop_Att
 maxDistance=3000
 soundClassObject=&Fort_Loot_Collect
 volumeMultiplier=0.6
}
