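# Sound cue for dropping a Pizza Party slice item, built from the three Pizza Party box-drop recordings.
# An equal-weight Random node picks one wave and feeds it through a Modulator left at default settings.
# The cue raises pitch and lowers volume on the box sounds, and uses weapon-drop attenuation with the loot-collect sound class.
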SoundCue /FlipperGameplay/Items/PizzaParty/Audio/PizzaParty_Slice_Drop_Cue
root=Modulator_2
nodes [
 {id=Modulator_2, children [Random_3]}
 {id=Random_3, children [WavePlayer_10, WavePlayer_11, WavePlayer_12], properties {weights [1,1,1]}}
 {id=WavePlayer_10, properties {soundWaveAssetPtr=&/PizzaParty_BoxDrop_01}}
 {id=WavePlayer_11, properties {soundWaveAssetPtr=&/PizzaParty_BoxDrop_02}}
 {id=WavePlayer_12, properties {soundWaveAssetPtr=&/PizzaParty_BoxDrop_03}}
]
properties {
 attenuationSettings=&Fort_Weapon_Drop_Att
 maxDistance=3000
 pitchMultiplier=1.3
 soundClassObject=&Fort_Loot_Collect
 volumeMultiplier=0.4
}
