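# Sound cue for equipping a Pizza Party slice item.
# A Random node picks one of four equally weighted equip waves, and a Modulator above it lowers pitch slightly at random for variety.
# It plays at reduced volume with weapon-reload attenuation and the loot-collect sound class, with a concurrency override capping simultaneous instances.
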SoundCue /FlipperGameplay/Items/PizzaParty/Audio/PizzaParty_Slice_Equip_Cue
root=Modulator_0
nodes [
 {id=Modulator_0, children [Random_0], properties {pitchMax=1, pitchMin=0.9}}
 {id=Random_0, children [WavePlayer_0,WavePlayer_2,WavePlayer_3,WavePlayer_4], properties {weights [1,1,1,1]}}
 {id=WavePlayer_0, properties {soundWaveAssetPtr=&/PizzaParty_SliceEquip_01}}
 {id=WavePlayer_2, properties {soundWaveAssetPtr=&/PizzaParty_SliceEquip_02}}
 {id=WavePlayer_3, properties {soundWaveAssetPtr=&/PizzaParty_SliceEquip_03}}
 {id=WavePlayer_4, properties {soundWaveAssetPtr=&/PizzaParty_SliceEquip_04}}
]
properties {
 attenuationSettings=&Fort_Weapon_Reload_Att
 bOverrideConcurrency=true
 maxDistance=2250
 soundClassObject=&Fort_Loot_Collect
 volumeMultiplier=0.6
 concurrencyOverrides {maxCount=3}
}
