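# Looping sound cue for a Pizza Party slice in the air.
# A Mixer layers a borrowed flying disc toy loop, held at reduced volume and fixed pitch, with a pizza slice loop through a default Modulator.
# Playback uses the gadgets sound class with overridden logarithmic attenuation.
# Overridden concurrency caps instances per owner and applies volume scaling.
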SoundCue /FlipperGameplay/Items/PizzaParty/Audio/PizzaParty_Slice_InAirLoop_Cue
root=Mixer_0
nodes [
 {id=Mixer_0, children [Modulator_0,Modulator_1], properties {inputVolume [1,1]}}
 {id=Modulator_0, children [WavePlayer_0], properties {pitchMax=1, pitchMin=1, volumeMax=0.7, volumeMin=0.7}}
 {id=Modulator_1, children [WavePlayer_1]}
 {id=WavePlayer_0, properties {bLooping=true, soundWaveAssetPtr=/Game/Athena/Sounds/Toys/FlyingDisc/Athena_Toy_FlyingDisc_InAirLoop_02}}
 {id=WavePlayer_1, properties {bLooping=true, soundWaveAssetPtr=&/PizzaParty_SliceInAirLoop_01}}
]
properties {
 attenuationSettings=/Game/Sounds/Attenuation/Fort_ToyHit_Att
 bOverrideAttenuation=true
 bOverrideConcurrency=true
 maxDistance=2000
 soundClassObject=&Fort_Gadgets
 volumeMultiplier=0.6
 attenuationOverrides {distanceAlgorithm=Logarithmic, falloffDistance=2000, stereoSpread=512, attenuationShapeExtents {x=0}}
 concurrencyOverrides {bLimitToOwner=true, maxCount=2, volumeScale=0.7}
}
