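# Sound cue for the Pizza Party item, played by its name when a thrown pizza slice lands.
# A Random node picks one of five equally weighted slice-impact waves and feeds a Modulator that varies pitch slightly around normal.
# It borrows the Corn bounce attenuation, uses the ranged-weapon projectile effects sound class, and overrides concurrency to cap simultaneous plays, blocking new ones at the limit.
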SoundCue /FlipperGameplay/Items/PizzaParty/Audio/PizzaParty_Slice_Landed_Cue
root=Modulator_0
nodes [
 {id=Modulator_0, children [Random_0], properties {pitchMax=0.9, pitchMin=1.1}}
 {id=Random_0, children [WavePlayer_0,WavePlayer_1,WavePlayer_2,WavePlayer_3,WavePlayer_4], properties {weights [1,1,1,1,1]}}
 {id=WavePlayer_0, properties {soundWaveAssetPtr=&/PizzaParty_SliceThrowImpact_01}}
 {id=WavePlayer_1, properties {soundWaveAssetPtr=&/PizzaParty_SliceThrowImpact_02}}
 {id=WavePlayer_2, properties {soundWaveAssetPtr=&/PizzaParty_SliceThrowImpact_03}}
 {id=WavePlayer_3, properties {soundWaveAssetPtr=&/PizzaParty_SliceThrowImpact_04}}
 {id=WavePlayer_4, properties {soundWaveAssetPtr=&/PizzaParty_SliceThrowImpact_05}}
]
properties {
 attenuationSettings=/Game/Athena/Sounds/Items/Corn/Fort_Corn_Bounce_Att
 bOverrideConcurrency=true
 maxDistance=3000
 soundClassObject=&Fort_WeaponRanged_Projectile_Effects
 volumeMultiplier=1
 concurrencyOverrides {maxCount=3, resolutionRule=PreventNew}
}
