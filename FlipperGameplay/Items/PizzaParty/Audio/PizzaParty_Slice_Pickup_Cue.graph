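# Sound Cue for picking up a pizza slice from the Pizza Party item, using the loot-collect sound class and weapon-drop attenuation.
# A mixer layers a slightly delayed random pick of four equally weighted slice pickup waves with a louder-weighted, pitched-up player melee equip sound.
# A root modulator at default settings processes the combined mix, and the cue's overall volume is halved.
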SoundCue /FlipperGameplay/Items/PizzaParty/Audio/PizzaParty_Slice_Pickup_Cue
root=Modulator_1
nodes [
 {id=Delay_1, children [Random_0], properties {delayMax=0.12, delayMin=0.12}}
 {id=Mixer_1, children [Delay_1,Modulator_2], properties {inputVolume [0.8,1.6]}}
 {id=Modulator_1, children [Mixer_1]}
 {id=Modulator_2, children [WavePlayer_6], properties {pitchMax=1.1, pitchMin=1.1}}
 {id=Random_0, children [WavePlayer_0,WavePlayer_3,WavePlayer_4,WavePlayer_5], properties {weights [1,1,1,1]}}
 {id=WavePlayer_0, properties {soundWaveAssetPtr=&/PizzaParty_SlicePickup_01}}
 {id=WavePlayer_3, properties {soundWaveAssetPtr=&/PizzaParty_SlicePickup_02}}
 {id=WavePlayer_4, properties {soundWaveAssetPtr=&/PizzaParty_SlicePickup_03}}
 {id=WavePlayer_5, properties {soundWaveAssetPtr=&/PizzaParty_SlicePickup_04}}
 {id=WavePlayer_6, properties {soundWaveAssetPtr=/Game/Sounds/Fort_GamePlay_Sounds/Player/player_melee_equip_02}}
]
properties {
 attenuationSettings=&Fort_Weapon_Drop_Att
 maxDistance=3000
 soundClassObject=&Fort_Loot_Collect
 volumeMultiplier=0.5
}
