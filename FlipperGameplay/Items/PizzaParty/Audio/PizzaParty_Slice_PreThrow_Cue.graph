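# Sound cue for the pre-throw foley of a Pizza Party slice item.
# A mixer layers one of two random slice pre-throw waves with one of two random box pre-throw waves, the box layer at lower volume.
# A modulator adds slight pitch variation to the mix.
# It plays quietly in the weapon foley sound class, using weapon reload attenuation.
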SoundCue /FlipperGameplay/Items/PizzaParty/Audio/PizzaParty_Slice_PreThrow_Cue
root=Modulator_0
nodes [
 {id=Mixer_0, children [Random_2,Random_3], properties {inputVolume [1,0.6]}}
 {id=Modulator_0, children [Mixer_0], properties {pitchMax=1.02, pitchMin=0.98}}
 {id=Random_2, children [WavePlayer_4,WavePlayer_5], properties {weights [1,1]}}
 {id=Random_3, children [WavePlayer_0,WavePlayer_2], properties {weights [1,1]}}
 {id=WavePlayer_0, properties {soundWaveAssetPtr=&/PizzaParty_BoxPreThrow_03}}
 {id=WavePlayer_2, properties {soundWaveAssetPtr=&/PizzaParty_BoxPreThrow_04}}
 {id=WavePlayer_4, properties {soundWaveAssetPtr=&/PizzaParty_SlicePreThrow_01}}
 {id=WavePlayer_5, properties {soundWaveAssetPtr=&/PizzaParty_SlicePreThrow_02}}
]
properties {
 attenuationSettings=&Fort_Weapon_Reload_Att
 maxDistance=2250
 soundClassObject=&Fort_WeaponFoley
 volumeMultiplier=0.125
}
