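# Sound cue for the projectile-by whip of a PizzaParty pizza slice, judging by its name and wave assets.
# A Random node picks one of three equally weighted slice whip waves and feeds it through a Modulator at default settings.
# It uses a launcher in-air projectile attenuation asset and a third-person ranged-weapon fire sound class, with reduced volume and a concurrency override that prevents new instances at the limit.
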SoundCue /FlipperGameplay/Items/PizzaParty/Audio/PizzaParty_Slice_Prox_ProjectileBy_Cue
root=Modulator_0
nodes [
 {id=Modulator_0, children [Random_1]}
 {id=Random_1, children [WavePlayer_0,WavePlayer_4,WavePlayer_5], properties {weights [1,1,1]}}
 {id=WavePlayer_0, properties {soundWaveAssetPtr=&/PizzaParty_SliceWhip_01}}
 {id=WavePlayer_4, properties {soundWaveAssetPtr=&/PizzaParty_SliceWhip_02}}
 {id=WavePlayer_5, properties {soundWaveAssetPtr=&/PizzaParty_SliceWhip_03}}
]
properties {
 attenuationSettings=/Game/Sounds/Attenuation/Fort_Weapon_Launcher_Proj_InAir_Default_Att
 bOverrideConcurrency=true
 maxDistance=12750
 soundClassObject=&Fort_WeaponRanged_Fire_3P_PSM
 volumeMultiplier=0.6
 attenuationOverrides {falloffDistance=5000, attenuationShapeExtents {x=1000}}
 concurrencyOverrides {resolutionRule=PreventNew}
}
