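# Sound cue for throwing a Pizza Party slice, routed to the gadgets sound class with gadget-placement attenuation.
# A Mixer blends a random pick of two main throw waves, behind a Delay node left at default settings, with a half-volume random pick of three layer waves.
# A root Modulator varies the mix, and a concurrency override prevents new instances once the limit is reached.
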SoundCue /FlipperGameplay/Items/PizzaParty/Audio/PizzaParty_Slice_Throw_Cue
root=Modulator_0
nodes [
 {id=Delay_0, children [Random_0]}
 {id=Mixer_0, children [Delay_0,Random_1], properties {inputVolume [1,0.5]}}
 {id=Modulator_0, children [Mixer_0]}
 {id=Random_0, children [WavePlayer_0,WavePlayer_2], properties {weights [1,1]}}
 {id=Random_1, children [WavePlayer_3,WavePlayer_4,WavePlayer_5], properties {weights [1,1,1]}}
 {id=WavePlayer_0, properties {soundWaveAssetPtr=&/PizzaParty_SliceThrow_01}}
 {id=WavePlayer_2, properties {soundWaveAssetPtr=&/PizzaParty_SliceThrow_03}}
 {id=WavePlayer_3, properties {soundWaveAssetPtr=&/PizzaParty_SliceThrowLayer_02}}
 {id=WavePlayer_4, properties {soundWaveAssetPtr=&/PizzaParty_SliceThrowLayer_03}}
 {id=WavePlayer_5, properties {soundWaveAssetPtr=&/PizzaParty_SliceThrowLayer_01}}
]
properties {
 attenuationSettings=/Game/Sounds/Attenuation/Fort_GadgetPlace_Att
 bOverrideConcurrency=true
 maxDistance=2000
 soundClassObject=/Game/Sounds/FortSoundClassesAndModes/Fort_Gadgets_StW
 volumeMultiplier=1.1
 concurrencyOverrides {resolutionRule=PreventNew}
}
