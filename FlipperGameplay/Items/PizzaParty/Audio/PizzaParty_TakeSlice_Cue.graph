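# Sound cue played when a pizza slice is taken from the Pizza Party item.
# A mixer layers a randomly chosen take-slice sound with a quieter random slice-pickup sound from four variants.
# The pickup layer is briefly delayed and pitched up by a modulator.
# It uses weapon-pickup attenuation and the gadget items sound class, with overridden concurrency that prevents new instances when the limit is reached.
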SoundCue /FlipperGameplay/Items/PizzaParty/Audio/PizzaParty_TakeSlice_Cue
root=Mixer_0
nodes [
 {id=Delay_0, children [Random_2], properties {delayMax=0.1, delayMin=0.1}}
 {id=Mixer_0, children [Random_1,Modulator_0], properties {inputVolume [1,0.5]}}
 {id=Modulator_0, children [Delay_0], properties {pitchMax=1.15, pitchMin=1.25}}
 {id=Random_1, children [WavePlayer_1,WavePlayer_5], properties {weights [1,1]}}
 {id=Random_2, children [WavePlayer_4,WavePlayer_6,WavePlayer_7,WavePlayer_8], properties {weights [1,1,1,1]}}
 {id=WavePlayer_1, properties {soundWaveAssetPtr=&/PizzaParty_TakeSlice_01}}
 {id=WavePlayer_4, properties {soundWaveAssetPtr=&/PizzaParty_SlicePickup_01}}
 {id=WavePlayer_5, properties {soundWaveAssetPtr=&/PizzaParty_TakeSlice_03}}
 {id=WavePlayer_6, properties {soundWaveAssetPtr=&/PizzaParty_SlicePickup_02}}
 {id=WavePlayer_7, properties {soundWaveAssetPtr=&/PizzaParty_SlicePickup_03}}
 {id=WavePlayer_8, properties {soundWaveAssetPtr=&/PizzaParty_SlicePickup_04}}
]
properties {
 attenuationSettings=&Fort_Weapon_Pickup_Att
 bOverrideConcurrency=true
 maxDistance=3500
 soundClassObject=&Fort_Gadgets_Items
 volumeMultiplier=0.5
 concurrencyOverrides {resolutionRule=PreventNew}
}
